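# Post-process AnimBP for the FoundKelpUniversal cape backpack on the Base_Cape_v2 skeleton, correcting its pose by game context.
# Chains additive bone-space offsets on backpack_root and two cape master bones, each weighted by a 0/1 float.
# EventGraph sets these weights from front end, skydiving, LOD above 1, and front end with cloth at LOD0.
# An OnInitialUpdate rigid-body override function exists, but the AnimGraph shows no rigid-body node.
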AnimBlueprint /BRCosmetics/Accessories/FORT_Backpacks/Backpack_M_MED_FoundKelpUniversal/Meshes/M_MED_FoundKelpUniversal_Cape_PostProcess_AnimBP
target_skeleton=/Game/Accessories/FORT_Capes/Common/Base_Cape_v2/Fortnite_Base_Cape_v2_Skeleton
graphs [
 AnimGraph {nodes [
   n1 = "Transform (Modify) Bone - Bone: backpack_root"(
    ComponentPose=Local To Component(LocalPose=Input Pose())
    Alpha=Get FE() -> (FE:double)
    Translation="2,-2,0"
    Rotation="2,0,-2.5"
    Scale="1,1,1"
   ) {
    rotationMode=BMM_Additive
    rotationSpace=BCS_BoneSpace
    translationMode=BMM_Additive
    translationSpace=BCS_BoneSpace
    boneToModify {boneName=backpack_root}
    rotation {pitch=2, roll=-2.5}
    translation=<2 -2>
   }
   n2 = Get LOD0() -> (LOD0:double)
   n3 = "Transform (Modify) Bone - Bone: dyn_cape_MSTR"(ComponentPose=@n1, Alpha=@n2, Translation="0,0,0", Rotation="-10,0,0", Scale="1,1,1") {
    rotationMode=BMM_Additive
    rotationSpace=BCS_BoneSpace
    translationMode=BMM_Additive
    translationSpace=BCS_BoneSpace
    boneToModify {boneName=dyn_cape_MSTR}
    rotation {pitch=-10}
   }
   n4 = "Transform (Modify) Bone - Bone: dyn_cape_r_MSTR"(ComponentPose=@n3, Alpha=@n2, Translation="0,0,0", Rotation="-5,0,0", Scale="1,1,1") {
    rotationMode=BMM_Additive
    rotationSpace=BCS_BoneSpace
    translationMode=BMM_Additive
    translationSpace=BCS_BoneSpace
    boneToModify {boneName=dyn_cape_r_MSTR}
    rotation {pitch=-5}
   }
   n5 = "Transform (Modify) Bone - Bone: backpack_root"(ComponentPose=@n4, Alpha=Get LOD_not_0_1() -> (LOD_not_0_1:double), Translation="0,0,0", Rotation="0,0,-3", Scale="1,1,1") {
    rotationMode=BMM_Additive
    rotationSpace=BCS_BoneSpace
    translationMode=BMM_Additive
    translationSpace=BCS_BoneSpace
    boneToModify {boneName=backpack_root}
    rotation {roll=-3}
   }
   n6 = "Transform (Modify) Bone - Bone: backpack_root"(ComponentPose=@n5, Alpha=Get skydive() -> (skydive:double), Translation="7,-2,0", Rotation="0,0,0", Scale="1,1,1") {translationMode=BMM_Additive, translationSpace=BCS_BoneSpace, boneToModify {boneName=backpack_root}, translation=<7 -2>}
   Output Pose(Result=Component To Local(ComponentPose=@n6)) {layerGroup=None}
  ]}
 EventGraph {nodes [
   Event BlueprintUpdateAnimation()
   > Set skydive(skydive:double="To Float (Boolean)"(InBool=Get bIsSkyDiving())) -> (Output_Get:double)
   n1 = Get CurrentLOD()
   n2 = "Equal (Integer)"(A=@n1)
   n3 = Get bClothEnabled()
   n4 = Get bIsInFrontEnd()
   > Set LOD0(LOD0:double="To Float (Boolean)"(InBool=AND Boolean!(A=@n3, B=@n2, C:bool=@n4) -> (ReturnValue))) -> (Output_Get:double)
   > n5 = Branch(Condition=OR Boolean(A=@n2, B="Equal (Integer)"(A=@n1, B=1)))
   > n6 = Set LOD_not_0_1(LOD_not_0_1:double=0) -> (Output_Get:double)
   Set FE(execute=[@n6.then, @n7.then], FE:double="To Float (Boolean)"(InBool=Get bIsInFrontEnd())) -> (Output_Get:double)
   n7 = Set LOD_not_0_1(execute=@n5.else, LOD_not_0_1=1)
  ]}
 OnInitialUpdate {nodes [
   n1 = OnInitialUpdate!() -> (then, Context:"Anim Update Context Structure (by ref)", Node:"Anim Node Reference Structure (by ref)") {is=title K2Node_FunctionEntry}
   > n2 = ConvertToRigidBodyAnimNode!(execute, Node:"Anim Node Reference Structure (by ref)"=@n1.Node) -> (Succeeded:Exec, Failed:Exec, ReturnValue:Rigid Body Anim Node Reference Structure)
   > SetOverridePhysicsAsset!(execute, Node:"Rigid Body Anim Node Reference Structure (by ref)"=@n2, PhysicsAsset:Physics Asset Object Reference) -> (then, ReturnValue:Rigid Body Anim Node Reference Structure)
   > Return Node()
  ]}
]
variables [
 FE {type=number}
 LOD0 {type=number}
 LOD_not_0_1 {type=number}
 skydive {type=number}
]
